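# Physics asset for the BountyPond backpack cosmetic, giving its mesh a small chain of capsule bodies for physical motion.
# Capsules on base_bp, dyn_chest, dyn_neck and dyn_head form a spine, with base_string joined to the neck and branching into left and right string bodies.
# Pack-to-chest, chest-to-neck and neck-to-string-base joints rotate freely.
# Head and string-tip joints keep angular limits, with no twist range on the strings.
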PhysicsAsset /BRCosmetics/Accessories/FORT_Backpacks/Backpack_M_MED_BountyPond/Meshes/M_MED_BountyPond_Pack_Physics
bodies {
 base_bp [{
   shapeType=Capsule
   length=0.0193
   radius=3.8799
   shapeName=base_bp_capsule
   center=<-0.1578 -1.0048 -3.9976>
   rotation=<rot -8.8779 -90 -92.2836>
  }]
 base_string [{shapeType=Capsule, length=0, radius=0.505, shapeName=base_string_capsule, center=<-0.3878 -0.0261 0.0149>, rotation=<rot 0 3.0078 -87.1868>}]
 dyn_chest [{shapeType=Capsule, length=2.4114, radius=3.3163, shapeName=dyn_chest_capsule, center=<0.4274 -0.2968 0.958>, rotation=<rot 90 0 0>}]
 dyn_head [{shapeType=Capsule, length=4.9651, radius=4.8322, center=<2.1125 -0.1179 -0.5114>, rotation=<rot 0.7521 -90 8.7428>}]
 dyn_neck [{shapeType=Capsule, length=4.405, radius=2.0336, shapeName=dyn_neck_capsule, center=<2.0179 -0.1541 0.5461>, rotation=<rot 82.5872 180 180>}]
 dyn_string_l [{shapeType=Capsule, length=2, radius=0.2, shapeName=dyn_string_l_capsule, center=<1.3171 -0.0336 0.6252>, rotation=<rot 0 96.7755 -63.5399>}]
 dyn_string_r [{shapeType=Capsule, length=2, radius=0.2, shapeName=dyn_string_r_capsule, center=<1.105 -0.0193 0.274>, rotation=<rot 0 84.9961 -81.7399>}]
}
constraints [
 {bone1Name=dyn_head, bone2Name=dyn_neck, swing1LimitDegrees=45, swing2LimitDegrees=45, twistLimitDegrees=45}
 {
  bone1Name=dyn_chest
  bone2Name=base_bp
  swing1LimitDegrees=45
  swing1Motion=Free
  swing2LimitDegrees=45
  swing2Motion=Free
  twistLimitDegrees=45
  twistMotion=Free
 }
 {
  bone1Name=dyn_neck
  bone2Name=dyn_chest
  swing1LimitDegrees=45
  swing1Motion=Free
  swing2LimitDegrees=45
  swing2Motion=Free
  twistLimitDegrees=45
  twistMotion=Free
 }
 {
  bone1Name=base_string
  bone2Name=dyn_neck
  swing1LimitDegrees=45
  swing1Motion=Free
  swing2LimitDegrees=45
  swing2Motion=Free
  twistLimitDegrees=45
  twistMotion=Free
 }
 {bone1Name=dyn_string_l, bone2Name=base_string, swing1LimitDegrees=25, swing2LimitDegrees=25, twistLimitDegrees=0}
 {bone1Name=dyn_string_r, bone2Name=base_string, swing1LimitDegrees=25, swing2LimitDegrees=25, twistLimitDegrees=0}
]
