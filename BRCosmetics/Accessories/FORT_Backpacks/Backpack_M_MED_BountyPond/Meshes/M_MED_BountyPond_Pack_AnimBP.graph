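# Anim Blueprint for the BountyPond backpack cosmetic, playing idle and reactive animations triggered by enemy eliminations in-game or a delayed, repeating timer in the frontend.
# The cached state-machine pose gets a spine-driven counter-rotation on base_bp, swapped for a fixed pack offset during DBNO or emotes.
# AnimDynamics springs then move three foliage bones, followed by RigidBody simulation that the DisableDynamics curve fades out.
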AnimBlueprint /BRCosmetics/Accessories/FORT_Backpacks/Backpack_M_MED_BountyPond/Meshes/M_MED_BountyPond_Pack_AnimBP
target_skeleton=&/M_MED_BountyPond_Pack_Skeleton
graphs [
 AnimGraph {nodes [
   n1 = "Transform (Modify) Bone - Bone: base_bp"!(
    ComponentPose=Local To Component(LocalPose=Use cached pose 'StartCache'())
    bAlphaBoolEnabled:bool=True
    Translation="0,0,0"
    Rotation="0,0,0"
    Scale="1,1,1"
   ) -> (Pose) {
    alphaInputType=Bool
    rotationMode=BMM_Additive
    rotationSpace=BCS_BoneSpace
    alphaBoolBlend {blendOption=Linear}
    boneToModify {boneName=base_bp}
   }
   n2 = "Transform (Modify) Bone - Bone: pack"!(ComponentPose=@n1, bAlphaBoolEnabled:bool, Translation="0,-2.5,-5", Rotation="0,0,0", Scale) -> (Pose) {
    alphaInputType=Bool
    rotationMode=BMM_Additive
    rotationSpace=BCS_BoneSpace
    translationMode=BMM_Additive
    translationSpace=BCS_BoneSpace
    alphaBoolBlend {blendOption=Linear}
    boneToModify {boneName=pack}
    translation {y=-2.5, z=-5}
   }
   n3 = Component To Local(ComponentPose=@n2)
   n4 = Get DBNOorEmote() -> (DBNOorEmote:bool)
   n5 = "Transform (Modify) Bone - Bone: base_bp"!(
    ComponentPose=Local To Component(LocalPose=Use cached pose 'StartCache'())
    Alpha:float=1
    Translation=Get BP Translate() -> (BP Translate:Vector)
    Rotation=Get Gimbal Rot() -> (Gimbal Rot:Rotator)
    Scale="1,1,1"
   ) -> (Pose) {
    rotationMode=BMM_Additive
    rotationSpace=BCS_BoneSpace
    alphaBoolBlend {blendOption=Linear}
    alphaScaleBiasClamp {bInterpResult=true}
    boneToModify {boneName=base_bp}
   }
   n6 = "Transform (Modify) Bone - Bone: base_bp"!(ComponentPose=@n5, bAlphaBoolEnabled:bool=True, Translation="0,0,0", Rotation="0,0,0", Scale="1,1,1") -> (Pose) {
    alphaInputType=Bool
    rotationMode=BMM_Additive
    rotationSpace=BCS_BoneSpace
    alphaBoolBlend {blendOption=QuarticInOut, blendOutTime=10}
    alphaScaleBiasClamp {bInterpResult=true}
    boneToModify {boneName=base_bp}
   }
   n7 = Component To Local(ComponentPose=@n6)
   n8 = "Transform (Modify) Bone - Bone: base_bp"!(
    ComponentPose=Local To Component(LocalPose=Blend Poses by bool(BlendPose_0=@n3, BlendPose_1=@n7, bActiveValue=@n4) {blendType=Linear})
    Alpha:float=1
    Translation=Get BP Translate()
    Rotation=Get BP Rotate() -> (BP Rotate:Rotator)
    Scale="1,1,1"
   ) -> (Pose) {
    rotationMode=BMM_Additive
    rotationSpace=BCS_BoneSpace
    translationMode=BMM_Additive
    translationSpace=BCS_BoneSpace
    alphaBoolBlend {blendInTime=2, blendOption=QuarticInOut, blendOutTime=5}
    alphaScaleBiasClamp {bInterpResult=true}
    boneToModify {boneName=base_bp}
   }
   n9 = "Transform (Modify) Bone - Bone: base_bp"!(ComponentPose=@n8, bAlphaBoolEnabled:bool=True, Translation="-1,0,0", Rotation="0,0,15", Scale="1,1,1") -> (Pose) {
    alphaInputType=Bool
    rotationMode=BMM_Additive
    rotationSpace=BCS_BoneSpace
    translationMode=BMM_Additive
    translationSpace=BCS_BoneSpace
    alphaBoolBlend {blendInTime=2, blendOption=QuarticInOut, blendOutTime=5}
    alphaScaleBiasClamp {bInterpResult=true}
    boneToModify {boneName=base_bp}
    rotation {roll=15}
    translation {x=-1}
   }
   n10 = Anim Dynamics - Bone: dyn_foliage_l!(ComponentPose=@n9, bAlphaBoolEnabled:bool=True, GravityOverride:Vector="0,0,-980", ExternalForce="0,0,0") -> (Pose) {
    alphaInputType=Bool
    angularDampingOverride=1
    angularSpringConstant=20000
    bAngularSpring=true
    bOverrideAngularDamping=true
    bUseGravityOverride=true
    gravityScale=0.35
    simulationSpace=RootRelative
    alphaBoolBlend {blendInTime=6, blendOption=Cubic, blendOutTime=6}
    alphaScaleBiasClamp {bInterpResult=true}
    boundBone {boneName=dyn_foliage_l}
    componentAppliedLinearAccClamp=<1500 1500 1500>
    componentLinearAccScale=<1 1 1>
    componentLinearVelScale=<1 1 1>
    gravityOverride {z=-980}
    physicsBodyDefinitions [{
      boundBone {boneName=dyn_foliage_l}
      boxExtents=<20 30 30>
      constraintSetup {angularLimitsMax=<0 0 4>, angularLimitsMin=<0 0 -4>, angularTarget=<1 0 0>}
      localJointOffset=<4 0 0>
     }]
    simSpaceSettings {maxAngularAcceleration=400, maxAngularVelocity=100, simSpaceAngularAlpha=0.25}
   }
   n11 = Anim Dynamics - Bone: dyn_foliage_m!(ComponentPose=@n10, bAlphaBoolEnabled:bool=True, GravityOverride:Vector="0,0,-980", ExternalForce="0,0,0") -> (Pose) {
    alphaInputType=Bool
    angularDampingOverride=1
    angularSpringConstant=20000
    bAngularSpring=true
    bOverrideAngularDamping=true
    bUseGravityOverride=true
    gravityScale=0.35
    simulationSpace=RootRelative
    alphaBoolBlend {blendInTime=6, blendOption=Cubic, blendOutTime=6}
    alphaScaleBiasClamp {bInterpResult=true}
    boundBone {boneName=dyn_foliage_m}
    componentAppliedLinearAccClamp=<1500 1500 1500>
    componentLinearAccScale=<1 1 1>
    componentLinearVelScale=<1 1 1>
    gravityOverride {z=-980}
    physicsBodyDefinitions [{
      boundBone {boneName=dyn_foliage_m}
      boxExtents=<20 30 30>
      constraintSetup {angularLimitsMax=<0 0 4>, angularLimitsMin=<0 0 -4>, angularTarget=<1 0 0>}
      localJointOffset=<4 0 0>
     }]
    simSpaceSettings {maxAngularAcceleration=400, maxAngularVelocity=100, simSpaceAngularAlpha=0.25}
   }
   n12 = Anim Dynamics - Bone: dyn_foliage_r!(ComponentPose=@n11, bAlphaBoolEnabled:bool=True, GravityOverride:Vector="0,0,-980", ExternalForce="0,0,0") -> (Pose) {
    alphaInputType=Bool
    angularDampingOverride=1
    angularSpringConstant=20000
    bAngularSpring=true
    bOverrideAngularDamping=true
    bUseGravityOverride=true
    gravityScale=0.35
    simulationSpace=RootRelative
    alphaBoolBlend {blendInTime=6, blendOption=Cubic, blendOutTime=6}
    alphaScaleBiasClamp {bInterpResult=true}
    boundBone {boneName=dyn_foliage_r}
    componentAppliedLinearAccClamp=<1500 1500 1500>
    componentLinearAccScale=<1 1 1>
    componentLinearVelScale=<1 1 1>
    gravityOverride {z=-980}
    physicsBodyDefinitions [{
      boundBone {boneName=dyn_foliage_r}
      boxExtents=<20 30 30>
      constraintSetup {angularLimitsMax=<0 0 4>, angularLimitsMin=<0 0 -4>, angularTarget=<1 0 0>}
      localJointOffset=<4 0 0>
     }]
    simSpaceSettings {maxAngularAcceleration=400, maxAngularVelocity=100, simSpaceAngularAlpha=0.25}
   }
   n13 = RigidBody!(ComponentPose=@n12, AlphaCurveName:Name=DisableDynamics, OverrideWorldGravity:Vector="0,0,-980", ExternalForce) -> (Pose) {
    alphaInputType=Curve
    bClampLinearTranslationLimitToRefPose=true
    bOverrideWorldGravity=true
    overridePhysicsAsset=&/M_MED_BountyPond_Pack_Physics
    simulationSpace=BaseBoneSpace
    alphaScaleBiasClamp {bias=1, scale=-1}
    baseBoneRef {boneName=pack}
    componentAppliedLinearAccClamp=<0 0 0>
    simSpaceSettings {
     dampingAlpha=0
     maxAngularAcceleration=1000
     maxAngularVelocity=5
     maxLinearAcceleration=700
     maxLinearVelocity=800
     velocityScaleZ=3
     worldAlpha=1
     externalLinearDragV=<0.2 0.2>
    }
   }
   Output Pose(Result=Component To Local(ComponentPose=@n13)) {layerGroup=None}
   AnimGraphNode_StateMachine_0 = FE_IdleNReactive() {is=title AnimGraphNode_StateMachine}
   AnimGraphNode_StateMachine_1 = IdleNReactive() {is=title AnimGraphNode_StateMachine}
   Save cached pose 'StartCache'(Pose=Blend Poses by bool(
     BlendPose_0=@AnimGraphNode_StateMachine_0
     BlendPose_1=@AnimGraphNode_StateMachine_1
     bActiveValue=Get bShouldDisplayFrontendReactivity() -> (bShouldDisplayFrontendReactivity:bool)
    )) {cacheName=StartCache}
  ]}
 FE_IdleNReactive {
  entry=Idle
  path=AnimGraph.AnimGraphNode_StateMachine_0.FE_IdleNReactive
  ids=id
  nodes [
   AnimStateNode_0 = Idle()
   AnimStateNode_1 = Reactive() {is=title AnimStateNode, stateEntered {guid=#1, notifyName=Entered FE_Reactive_Anim}}
   AnimStateTransitionNode_0 = AnimStateTransitionNode() {bAllowInertializationForSelfTransitions=true}
   AnimStateTransitionNode_1 = AnimStateTransitionNode() {bAllowInertializationForSelfTransitions=true, bAutomaticRuleBasedOnSequencePlayerInState=true}
  ]
  links [
   {
    dst=Reactive
    rule=AnimGraph.AnimGraphNode_StateMachine_0.FE_IdleNReactive.AnimStateTransitionNode_0.Transition
    src=Idle
    via=AnimStateTransitionNode_0
   }
   {
    dst=Idle
    rule=AnimGraph.AnimGraphNode_StateMachine_0.FE_IdleNReactive.AnimStateTransitionNode_1.Transition
    src=Reactive
    via=AnimStateTransitionNode_1
   }
  ]
 }
 Idle {
  path=AnimGraph.AnimGraphNode_StateMachine_0.FE_IdleNReactive.AnimStateNode_0.Idle
  nodes [Output Animation Pose(Result=Sequence Player 'Idle_01_Pack'() {sequence=/BRCosmetics/Accessories/FORT_Backpacks/Backpack_M_MED_BountyPond/Anim/Idle_01_Pack})]
 }
 Reactive {
  path=AnimGraph.AnimGraphNode_StateMachine_0.FE_IdleNReactive.AnimStateNode_1.Reactive
  nodes [Output Animation Pose(Result=Sequence Player 'Reactive_Pack'() {bLoopAnimation=false, sequence=/BRCosmetics/Accessories/FORT_Backpacks/Backpack_M_MED_BountyPond/Anim/Reactive_Pack})]
 }
 Transition {
  path=AnimGraph.AnimGraphNode_StateMachine_0.FE_IdleNReactive.AnimStateTransitionNode_0.Transition
  nodes [Result(bCanEnterTransition=Get Play Reactive FE Anim() -> (Play Reactive FE Anim:bool))]
 }
 Transition {path=AnimGraph.AnimGraphNode_StateMachine_0.FE_IdleNReactive.AnimStateTransitionNode_1.Transition, nodes [Result()]}
 IdleNReactive {
  entry=Idle
  path=AnimGraph.AnimGraphNode_StateMachine_1.IdleNReactive
  ids=id
  nodes [
   AnimStateNode_0 = Idle()
   AnimStateNode_1 = Reactive()
   AnimStateTransitionNode_0 = AnimStateTransitionNode() {bAllowInertializationForSelfTransitions=true}
   AnimStateTransitionNode_1 = AnimStateTransitionNode() {bAllowInertializationForSelfTransitions=true, bAutomaticRuleBasedOnSequencePlayerInState=true}
  ]
  links [
   {
    dst=Reactive
    rule=AnimGraph.AnimGraphNode_StateMachine_1.IdleNReactive.AnimStateTransitionNode_0.Transition
    src=Idle
    via=AnimStateTransitionNode_0
   }
   {
    dst=Idle
    rule=AnimGraph.AnimGraphNode_StateMachine_1.IdleNReactive.AnimStateTransitionNode_1.Transition
    src=Reactive
    via=AnimStateTransitionNode_1
   }
  ]
 }
 Idle {
  path=AnimGraph.AnimGraphNode_StateMachine_1.IdleNReactive.AnimStateNode_0.Idle
  nodes [Output Animation Pose(Result=Sequence Player 'Idle_01_Pack'() {sequence=&Idle_01_Pack})]
 }
 Reactive {
  path=AnimGraph.AnimGraphNode_StateMachine_1.IdleNReactive.AnimStateNode_1.Reactive
  nodes [Output Animation Pose(Result=Sequence Player 'Reactive_Pack'() {bLoopAnimation=false, sequence=&Reactive_Pack})]
 }
 Transition {
  path=AnimGraph.AnimGraphNode_StateMachine_1.IdleNReactive.AnimStateTransitionNode_0.Transition
  nodes [Result(bCanEnterTransition=Get Play Reactive() -> (Play Reactive:bool))]
 }
 Transition {path=AnimGraph.AnimGraphNode_StateMachine_1.IdleNReactive.AnimStateTransitionNode_1.Transition, nodes [Result()]}
 EventGraph {nodes [
   n1 = Event BlueprintProcessAnimInputEvent()
   > Branch(Condition=IsNamed(self=@n1.AnimInputEvent, InEventName=aie.EnemyPlayerEliminated))
   > Set Play Reactive(Play Reactive:bool=true) -> (Output_Get:bool)
   > Delay(Duration=0.2)
   > Set Play Reactive(Play Reactive=false)
   > n2 = Branch(Condition=IsNamed(self=@n1.AnimInputEvent, InEventName=aie.Player.Zipline.Entered))
   > Set isOnZipline(isOnZipline:bool=1) -> (Output_Get:bool)
   Branch(execute=@n2.else, Condition=IsNamed(self=@n1.AnimInputEvent, InEventName=aie.Player.Zipline.Exited))
   > Set isOnZipline(isOnZipline=false)
   Event OnCharacterCustomizationLoaded()
   > n3 = Cast To FortPlayerPawn(Object=TryGetFortPawn())
   > n4 = Cast To B_Athena_PartModifier_Backpack_BountyPond!(
    execute
    Object=GetModifierForPartType!(self:Fort Player Pawn Object Reference=@n3, PartType:EFortCustomPartType Enum=Backpack) -> (ReturnValue:Custom Character Part Modifier Object Reference)
   ) -> (then, CastFailed, AsB Athena Part Modifier Backpack Bounty Pond:B Athena Part Modifier Backpack Bounty Pond Object Reference) {is=title K2Node_DynamicCast}
   > n5 = Branch(Condition=Get ReactivityEnabled!(self:B Athena Part Modifier Backpack Bounty Pond Object Reference=@n4) -> (ReactivityEnabled:bool))
   > Do Once(Reset=@n7.then)
   > n6 = Delay(Duration=4)
   Set Play Reactive FE Anim(execute=[@n17.then, @n6.then], Play Reactive FE Anim:bool=true) -> (Output_Get:bool)
   n7 = Set Play Reactive FE Anim(execute=@n5.else, Play Reactive FE Anim=false)
   Event BlueprintInitializeAnimation()
   > n8 = Cast To FortPlayerPawn(Object=TryGetFortPawn())
   > "Set Skel mesh "("Skel mesh ":Skeletal Mesh Component Object Reference=GetSkeletalMeshForPartType(self=@n8)) -> (Output_Get:Skeletal Mesh Component Object Reference)
   Event BlueprintPostEvaluateAnimation() {is=title K2Node_Event}
   > Set Play Reactive(Play Reactive=false)
   Event BlueprintUpdateAnimation()
   > n9 = Sequence!(execute) -> (then_0, then_1, then_2:Exec)
   n10 = GetOwnerSkeletalMeshComponent()
   n11 = BreakRotator(InRot=BreakTransform(InTransform=GetDeltaTransformFromRefPose(self=@n10, BoneName=spine_05, BaseName=spine_01)).Rotation)
   n12 = MapRangeClamped(Value=@n11.Yaw, InRangeA=0, InRangeB=-20, OutRangeB=-20)
   Set BP Rotate(
    execute=@n9.then_1
    BP Rotate:Rotator=MakeRotator(Yaw=FInterpEaseInOut!(A:double=0, B:double=@n12, Alpha:double=0, Exponent:double=0) -> (ReturnValue:double))
   ) -> (Output_Get:Rotator)
   n13 = MapRangeClamped(Value=@n11.Yaw, InRangeA=0, InRangeB=-30, OutRangeB=20)
   > Set BP Translate(BP Translate:Vector=MakeVector(Y=0, Z=FInterpEaseInOut(A=0, B=@n13, Alpha=0, Exponent=0))) -> (Output_Get:Vector)
   n14 = GetOwnerSkeletalMeshComponent()
   n15 = BreakRotator(InRot=BreakTransform(InTransform=GetDeltaTransformFromRefPose(self=@n14, BoneName=spine_05, BaseName=root)).Rotation)
   > Set Gimbal Rot(Gimbal Rot:Rotator=MakeRotator(Pitch=float * float(A=@n15.Yaw, B=-0.95) {is=title K2Node_PromotableOperator})) -> (Output_Get:Rotator)
   > Branch(Condition=NOT Boolean(A=Get bIsCrouching()))
   > Set ADSGimbal(ADSGimbal:Rotator=MakeRotator(Pitch=float * float(A=@n15.Yaw, B=0.3))) -> (Output_Get:Rotator)
   n16 = "Gravity Override State Function (Parameter Asset) (Nativized)"(
    execute=@n9.then_0
    CharacterPartAnimInstance=Self-Reference()
    GameTimeInSeconds=GetGameTimeInSeconds()
    Parameters=&/M_MED_BountyPond_Pack_Parameters
   )
   > Set Gravity Override(Gravity Override=@n16)
   > Set toggleAnimDynamics(toggleAnimDynamics:bool=NOT Boolean(A=Get isOnZipline() -> (isOnZipline:bool))) -> (Output_Get:bool)
   Branch(execute=@n9.then_2, Condition=NOT Boolean(A=Get bIsSkyDiving()))
   > Set Rot Down(Rot Down:float=GetAlphaFromReferenceTransformDelta(
     SkeletalMeshComponent="Get Skel mesh "() -> ("Skel mesh ":Skeletal Mesh Component Object Reference)
     BoneName=spine_03
     BaseName="head "
     InputRange=(X=0,Y=-120)
     OutputRange=(X=0,Y=-1)
     TwistAxis="0,0,1"
    )) -> (Output_Get:float)
   > Set Rot Up(Rot Up:float=GetAlphaFromReferenceTransformDelta(SkeletalMeshComponent="Get Skel mesh "(), BoneName=spine_04, BaseName=head, InputRange=(X=15,Y=60), TwistAxis="0,0,1")) -> (Output_Get:float)
   > Set DBNOorEmote(DBNOorEmote:bool=OR Boolean(A=Get bIsDBNO(), B=Get bIsPlayingEmote())) -> (Output_Get:bool)
   AnimNotify_Entered FE_Reactive_Anim() {is=title K2Node_Event}
   > Set Play Reactive FE Anim(Play Reactive FE Anim=false)
   > n17 = Delay(Duration=12)
   float * float(A=@n15.Yaw, B=-1)
   float * float(A=float * float(A=@n15.Yaw, B=Get Rot Down() -> (Rot Down:float)), B=0.5)
  ]}
]
variables [
 ADSGimbal {type=Rotator}
 BP Rotate {type=Rotator}
 BP Translate {type=Vector}
 DBNOorEmote {type=boolean}
 Gimbal Rot {type=Rotator}
 Gravity Override {type=Vector}
 isOnZipline {type=boolean}
 Play Reactive {type=boolean}
 Play Reactive FE Anim {type=boolean}
 Rot Down {type=number}
 Rot Up {type=number}
 "Skel mesh " {type=SkeletalMeshComponent}
 toggleAnimDynamics {type=boolean}
]
